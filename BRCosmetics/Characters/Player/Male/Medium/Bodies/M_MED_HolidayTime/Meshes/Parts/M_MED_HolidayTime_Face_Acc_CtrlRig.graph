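# Control Rig for the HolidayTime male medium face accessory that adjusts the dynamic beard chain as the head moves.
# It reads head rotation relative to spine_05, scales the Euler Z by -0.6, and adds it as local Y rotation on dyn_beard_01 to counter-rotate it.
# dyn_beard_02 then partially aims at dyn_beard_Aim, a bone parented to spine_05, propagating to its children.
# The hat bones are not driven here.
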ControlRigBlueprint /BRCosmetics/Characters/Player/Male/Medium/Bodies/M_MED_HolidayTime/Meshes/Parts/M_MED_HolidayTime_Face_Acc_CtrlRig
graphs [
 RigVMFunctionLibrary {links [], nodes []}
 RigVMModel {nodes [
   n1 = &RigUnit_BeginExecution()
   n2 = &RigVMFunction_MathQuaternionToEuler(RotationOrder=ZYX, Value=&RigUnit_GetRelativeBoneTransform(Bone=head, Space=spine_05).Transform.Rotation)
   /Script/ControlRig.RigUnit_ModifyBoneTransforms(
    BoneToModify=((Bone="dyn_beard_01",Transform=(Rotation=(X=0,Y=0,Z=0,W=1),Translation=(X=0,Y=0,Z=0),Scale3D=(X=1,Y=1,Z=1))))
    Mode=AdditiveLocal
    Weight=1
    WeightMaximum=1
    WeightMinimum=0
    ExecutePin=@n1.ExecutePin
    BoneToModify.0.Transform.Rotation=&RigVMFunction_MathQuaternionFromEuler(
     Euler=(X=0,Y=0,Z=0)
     Euler.Y=/Script/RigVM.RigVMFunction_MathVectorMul(A=(X=0,Y=0,Z=-0.6), B=@n2.Result) {is=struct RigVMUnitNode, template="Multiply::Execute(in A,in B,out Result)"}.Result.Z
    ).Result
   ) {is=struct RigVMUnitNode}
   > /Script/ControlRig.RigUnit_AimBone(
    Bone=dyn_beard_02
    DebugSettings=(bEnabled=true,Scale=10,WorldOffset=(Rotation=(X=0,Y=0,Z=0,W=1),Translation=(X=0,Y=0,Z=0),Scale3D=(X=1,Y=1,Z=1)))
    Primary=(Weight=0.5,Axis=(X=1,Y=0,Z=0),Target=(X=1,Y=0,Z=0),Kind=Location,Space="dyn_beard_Aim")
    Secondary=(Weight=0.3,Axis=(X=0,Y=1,Z=0),Target=(X=0,Y=1,Z=0),Kind=Direction,Space="dyn_beard_Aim")
    Weight=1
    bPropagateToChildren=True
   ) {is=struct RigVMUnitNode}
  ]}
]
hierarchy [
 root {loc=<0 0 0>, rot=<0 0 0 1>}
 attach {parent=root, loc=<0 0 0>, rot=<0 0 0 1>}
 pelvis {parent=root, loc=<0 -0.7071 93.1191>, rot=<0.0127 0.707 -0.0127 -0.707>}
 spine_01 {parent=pelvis, loc=<6.7494 -0.6388 0>, rot=<0 0 0.034 -0.9994>}
 spine_02 {parent=spine_01, loc=<7.2518 -0.035 0>, rot=<0 0 0.0238 -0.9997>}
 spine_03 {parent=spine_02, loc=<6.0104 0 0>, rot=<0 0 0.0149 -0.9999>}
 spine_04 {parent=spine_03, loc=<8.1024 0 0>, rot=<0 0 -0.0132 -0.9999>}
 spine_05 {parent=spine_04, loc=<10.0146 0 0>, rot=<0 0 -0.0015 -1>}
 clavicle_l {parent=spine_05, loc=<4.3575 -0.1466 -2.3223>, rot=<0.0465 -0.7289 -0.009 -0.6829>}
 clavicle_r {parent=spine_05, loc=<4.3575 -0.1466 2.3223>, rot=<0.7289 0.0465 -0.6829 0.009>}
 neck_01 {parent=spine_05, loc=<11.3138 0.1529 0>, rot=<0 0 -0.2003 -0.9797>}
 neck_02 {parent=neck_01, loc=<4.5627 0 0>, rot=<0 0 0.0504 -0.9987>}
 head {parent=neck_02, loc=<5.3914 0 0>, rot=<0 0 0.115 -0.9934>}
 faceAttach {parent=head, loc=<-0.8266 -0.1832 0>, rot=<0.0044 -0.3827 -0.0044 -0.9239>, scale=<0.89 0.89 0.89>}
 C_brow_mid {parent=faceAttach, loc=<4.7209 11.2084 4.8303>, rot=<-0.0586 -0.317 0.0114 -0.9466>}
 L_brow_mid {parent=faceAttach, loc=<7.639 10.371 3.157>, rot=<0.0603 -0.3852 0.2489 -0.8866>}
 L_brow_outer {parent=faceAttach, loc=<9.1042 8.1218 1.4857>, rot=<0.2349 -0.2525 0.3549 -0.869>}
 L_cheek_inner {parent=faceAttach, loc=<4.2785 8.761 -1.4435>, rot=<0.0454 -0.6262 0.3191 -0.7099>}
 C_lip_upper_mid {parent=faceAttach, loc=<-0.3154 10.7989 -0.2464>, rot=<0.2097 -0.3713 -0.0897 -0.9001>}
 L_lip_upper_outer {parent=faceAttach, loc=<0.8467 10.1076 -1.4454>, rot=<0.1857 -0.3758 0.3015 -0.8564>}
 L_lip_corner {parent=faceAttach, loc=<1.613 9.1457 -2.3124>, rot=<0.0012 -0.375 -0.0189 -0.9269>}
 C_nose_bridge {parent=faceAttach, loc=<2.3602 13.1515 2.4853>, rot=<-0.2002 -0.3676 0.0825 -0.9044>}
 R_brow_mid {parent=faceAttach, loc=<3.0903 10.342 7.7132>, rot=<-0.1347 -0.3556 -0.2231 -0.8976>}
 R_brow_outer {parent=faceAttach, loc=<1.4121 8.0675 9.1444>, rot=<-0.1251 -0.3913 -0.4176 -0.8104>}
 R_cheek_inner {parent=faceAttach, loc=<-1.4992 8.7305 4.3237>, rot=<-0.1979 -0.0545 -0.2598 -0.9436>}
 R_lip_corner {parent=faceAttach, loc=<-2.32 9.1134 1.692>, rot=<0.0127 -0.39 0.008 -0.9207>}
 C_jaw {parent=faceAttach, loc=<0.3123 3.2483 0.3583>, rot=<0.0044 -0.3827 -0.0044 -0.9239>}
 L_lip_lower_outer {parent=C_jaw, loc=<1.6966 6.8705 -0.8455>, rot=<-0.0146 -0.0279 0.3432 -0.9387>}
 C_lip_lower_mid {parent=C_jaw, loc=<0.0184 7.4191 -0.884>, rot=<0 -0.0099 -0.0021 -0.9999>}
 R_lip_lower_outer {parent=C_jaw, loc=<-1.6982 6.9042 -0.8853>, rot=<0.0021 -0.0171 -0.3152 -0.9489>}
 teeth_lower {parent=C_jaw, loc=<0.0169 5.3598 -2.8149>, rot=<0 0 0 1>}
 tongue {parent=C_jaw, loc=<0.0178 1.5245 -1.0398>, rot=<-0.1568 0 0 -0.9876>}
 R_eye {parent=faceAttach, loc=<2.2537 8.4351 6.6885>, rot=<0.3823 -0.0165 0.9223 -0.0548>}
 R_eye_lid_upper_mid {parent=R_eye, loc=<-0.0465 -1.169 0.2722>, rot=<0.0559 -0.0011 0.0197 -0.9982>, scale=<1 1.0001 0.9999>}
 R_eye_lid_lower_mid {parent=R_eye, loc=<-0.0141 -0.9451 -0.7006>, rot=<0.035 -0.0006 0.0197 -0.9992>}
 L_eye {parent=faceAttach, loc=<6.6314 8.4647 2.3108>, rot=<0.3819 0.0253 0.9227 0.046>}
 L_eye_lid_upper_mid {parent=L_eye, loc=<0.0251 -1.1623 0.2388>, rot=<0.0523 0 0 -0.9986>}
 L_eye_lid_lower_mid {parent=L_eye, loc=<0.0223 -0.923 -0.7488>, rot=<0.0329 0 0 -0.9995>}
 teeth_upper {parent=faceAttach, loc=<0.0954 9.3381 0.1585>, rot=<0.0044 -0.3827 -0.0044 -0.9239>}
 R_lip_upper_outer {parent=faceAttach, loc=<-1.497 10.1359 0.962>, rot=<-0.0848 -0.2926 -0.3215 -0.8966>}
 dyn_beard_01 {parent=head, loc=<4.9169 3.7821 0>, rot=<-0.135 -0.6941 -0.6935 -0.1382>}
 dyn_beard_02 {parent=dyn_beard_01, loc=<7.4874 0 0>, rot=<0.0005 0.1213 -0.0083 -0.9926>}
 dyn_beard_03 {parent=dyn_beard_02, loc=<7.6342 0 0>, rot=<0 -0.0042 0.0061 -1>}
 dyn_beard_04 {parent=dyn_beard_03, loc=<8.65 0 0>, rot=<0 0.0371 0 -0.9993>}
 dyn_beard_05 {parent=dyn_beard_04, loc=<7.9026 0 0>, rot=<0 0.2255 0 -0.9742>}
 dyn_beard_06 {parent=dyn_beard_05, loc=<7.0978 0 0>, rot=<0 0.577 0 -0.8168>}
 dyn_beard_07 {parent=dyn_beard_06, loc=<6.0731 0 0>, rot=<0 0.5555 0 -0.8315>}
 dyn_beard_08 {parent=dyn_beard_07, loc=<4.3308 0 0>, rot=<0 0.4965 0 -0.868>}
 dyn_beard_09 {parent=dyn_beard_08, loc=<3.3831 0 0>, rot=<0 0.5987 0 -0.801>}
 dyn_beard_10 {parent=dyn_beard_09, loc=<2.184 0 0>, rot=<0.0027 0.5632 -0.0055 -0.8263>}
 dyn_hat_01 {parent=head, loc=<11.3616 -0.0049 0>, rot=<0.554 -0.4394 0.4394 -0.554>}
 dyn_hat_02 {parent=dyn_hat_01, loc=<8.4608 0 0>, rot=<0 -0.2265 0 -0.974>}
 dyn_hat_03 {parent=dyn_hat_02, loc=<6.308 0 0>, rot=<0 -0.4023 0 -0.9155>}
 dyn_hat_04 {parent=dyn_hat_03, loc=<10.7357 0 0>, rot=<0 0.7579 0 -0.6523>}
 dyn_hat_05 {parent=dyn_hat_04, loc=<6.5821 0 0>, rot=<0 -0.7482 0 -0.6634>}
 dyn_hat_06 {parent=dyn_hat_05, loc=<6.7798 0 0>, rot=<0 0.3826 0 -0.9239>}
 dyn_beard_Aim {parent=spine_05, loc=<-0.5057 16.1325 -0.1234>, rot=<0.0284 -0.7069 -0.7062 0.0283>}
]
