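# Physics asset for the dynamic beard face accessory of the medium male Holiday Time cosmetic character.
# An eight-bone beard chain hangs off the head, using boxes that are wide and flat mid-chain and smaller at both ends.
# Upper-spine capsules provide torso shapes, joined to the head by free angular constraints, as is the beard root.
# Links down the beard allow limited swing with very tight twist, stiffening toward the tip.
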PhysicsAsset /BRCosmetics/Characters/Player/Male/Medium/Bodies/M_MED_HolidayTime/Meshes/Parts/M_MED_Holiday_Time_FaceAcc_DynBeard
bodies {
 dyn_beard_01 [{shapeType=Box, extentX=1, extentY=1, extentZ=1}]
 dyn_beard_02 [{shapeType=Box, extentX=1, extentY=1, extentZ=1}]
 dyn_beard_03 [{shapeType=Box, extentX=10, extentY=26, extentZ=3.5, center=<4 0 0>}]
 dyn_beard_04 [{shapeType=Box, extentX=8, extentY=25, extentZ=3.5, center=<4 0 0>}]
 dyn_beard_05 [{shapeType=Box, extentX=8, extentY=23, extentZ=3.5, center=<4 0 0>}]
 dyn_beard_06 [{shapeType=Box, extentX=6, extentY=6, extentZ=2, center=<3 0 0>}]
 dyn_beard_07 [{shapeType=Box, extentX=6, extentY=6, extentZ=2, center=<3 0 0>}]
 dyn_beard_08 [{shapeType=Box, extentX=4, extentY=3, extentZ=2, center=<2 0 0>}]
 head [{shapeType=Capsule, length=1, radius=1, rotation=<rot 90 0 0>}]
 spine_04 [{shapeType=Capsule, length=9.6662, radius=11.504, center=<4.4867 3.5164 0.0002>, rotation=<rot 0 -90 0>}]
 spine_05 [{shapeType=Capsule, length=7.6148, radius=13.0728, center=<1.6513 1.2713 0.0002>, rotation=<rot 0 -90 0>}]
}
constraints [
 {
  bone1Name=dyn_beard_01
  bone2Name=head
  swing1LimitDegrees=45
  swing1Motion=Free
  swing2LimitDegrees=45
  swing2Motion=Free
  twistLimitDegrees=45
  twistMotion=Free
 }
 {bone1Name=dyn_beard_02, bone2Name=dyn_beard_01, swing1LimitDegrees=30, swing2LimitDegrees=45, twistLimitDegrees=2}
 {bone1Name=dyn_beard_03, bone2Name=dyn_beard_02, swing1LimitDegrees=30, swing2LimitDegrees=60, twistLimitDegrees=2}
 {bone1Name=dyn_beard_04, bone2Name=dyn_beard_03, swing1LimitDegrees=30, swing2LimitDegrees=60, twistLimitDegrees=2}
 {bone1Name=dyn_beard_05, bone2Name=dyn_beard_04, swing1LimitDegrees=30, swing2LimitDegrees=60, twistLimitDegrees=2}
 {bone1Name=dyn_beard_06, bone2Name=dyn_beard_05, swing1LimitDegrees=30, swing2LimitDegrees=30, twistLimitDegrees=2}
 {bone1Name=dyn_beard_07, bone2Name=dyn_beard_06, swing1LimitDegrees=30, swing2LimitDegrees=20, twistLimitDegrees=2}
 {
  bone1Name=head
  bone2Name=spine_05
  swing1LimitDegrees=45
  swing1Motion=Free
  swing2LimitDegrees=45
  swing2Motion=Free
  twistLimitDegrees=45
  twistMotion=Free
 }
 {
  bone1Name=spine_05
  bone2Name=spine_04
  swing1LimitDegrees=45
  swing1Motion=Free
  swing2LimitDegrees=45
  swing2Motion=Free
  twistLimitDegrees=45
  twistMotion=Free
 }
 {bone1Name=dyn_beard_08, bone2Name=dyn_beard_07, swing1LimitDegrees=15, swing2LimitDegrees=10, twistLimitDegrees=2}
]
