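# Physics asset for the dynamic hat on the Holiday Time male medium character's face accessory, giving the hat a jointed chain for physical motion.
# Six capsule bodies on the dyn_hat bones shrink in radius toward the tip, with a small capsule on the head bone as the root.
# Each link is constrained to the previous one with limited swing, tighter on one axis near the head, and very little twist.
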PhysicsAsset /BRCosmetics/Characters/Player/Male/Medium/Bodies/M_MED_HolidayTime/Meshes/Parts/M_MED_Holiday_Time_FaceAcc_Dyn_hatBackling
bodies {
 dyn_hat_01 [{shapeType=Capsule, length=5, radius=5.4, center=<2.5 0 0>, rotation=<rot 90 0 0>}]
 dyn_hat_02 [{shapeType=Capsule, length=5, radius=4.5, center=<2.5 0 0>, rotation=<rot 90 0 0>}]
 dyn_hat_03 [{shapeType=Capsule, length=10, radius=2.8, center=<5 0 0>, rotation=<rot 90 0 0>}]
 dyn_hat_04 [{shapeType=Capsule, length=6, radius=2, center=<3 0 0>, rotation=<rot 90 0 0>}]
 dyn_hat_05 [{shapeType=Capsule, length=6, radius=1.5, center=<3 0 0>, rotation=<rot 90 0 0>}]
 dyn_hat_06 [{shapeType=Capsule, length=6, radius=1.5, center=<3 0 0>, rotation=<rot 90 0 0>}]
 head [{shapeType=Capsule, length=1, radius=1, rotation=<rot 90 0 0>}]
}
constraints [
 {bone1Name=dyn_hat_01, bone2Name=head, swing1LimitDegrees=30, swing2LimitDegrees=20, twistLimitDegrees=5}
 {bone1Name=dyn_hat_02, bone2Name=dyn_hat_01, swing1LimitDegrees=30, swing2LimitDegrees=20, twistLimitDegrees=5}
 {bone1Name=dyn_hat_03, bone2Name=dyn_hat_02, swing1LimitDegrees=30, swing2LimitDegrees=30, twistLimitDegrees=5}
 {bone1Name=dyn_hat_04, bone2Name=dyn_hat_03, swing1LimitDegrees=30, swing2LimitDegrees=30, twistLimitDegrees=5}
 {bone1Name=dyn_hat_05, bone2Name=dyn_hat_04, swing1LimitDegrees=30, swing2LimitDegrees=30, twistLimitDegrees=5}
 {bone1Name=dyn_hat_06, bone2Name=dyn_hat_05, swing1LimitDegrees=30, swing2LimitDegrees=30, twistLimitDegrees=5}
]
