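# Post-process style AnimBP for the Holiday Time face accessory, animating its hat and beard over the parent mesh's pose.
# Copied pose runs through a Control Rig and is cached.
# bRigidBodyEnabled selects RigidBody simulation, with backpack-dependent hat physics asset plus a beard pass, or a fallback of bone offsets, trail and spring controllers.
# Fallback alpha fades with pawn speed, with a fixed value for airborne or vehicle states.
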AnimBlueprint /BRCosmetics/Characters/Player/Male/Medium/Bodies/M_MED_HolidayTime/Meshes/Parts/M_MED_Holiday_Time_FaceAcc_AnimBP
target_skeleton=&/M_MED_Holiday_Time_FaceAcc_Skeleton
graphs [
 AnimGraph {nodes [
   n1 = "Transform (Modify) Bone - Bone: dyn_beard_03"(
    ComponentPose=Local To Component(LocalPose=Use cached pose 'pre pose'())
    Alpha=1
    Translation="0,0,0"
    Rotation="-8,0.000,0.000"
    Scale="1,1,1"
   ) {
    rotationMode=BMM_Additive
    rotationSpace=BCS_BoneSpace
    alphaBoolBlend {blendOption=Linear}
    boneToModify {boneName=dyn_beard_03}
    rotation {pitch=-8}
   }
   n2 = "Transform (Modify) Bone - Bone: dyn_beard_04"(ComponentPose=@n1, Alpha=1, Translation="0,0,0", Rotation="-26,0.000,0.000", Scale="1,1,1") {
    rotationMode=BMM_Additive
    rotationSpace=BCS_BoneSpace
    alphaBoolBlend {blendOption=Linear}
    boneToModify {boneName=dyn_beard_04}
    rotation {pitch=-26}
   }
   n3 = Get trail ctrl alpha() -> (trail ctrl alpha:double)
   n4 = Trail controller - Bone: dyn_hat_06(ComponentPose=@n2, Alpha=@n3) {
    bUsePlanarLimit=true
    chainLength=5
    relaxationSpeedScale=2.5
    alphaScaleBiasClamp {bInterpResult=true, interpSpeedDecreasing=3}
    planarLimits [{drivingBone {boneName=spine_05}, planeTransform {location=<18 0 0>, rotation=<rot 0 75 -89.9999>, scale=<1 1 1>}}]
    rotationLimits [{limitMax=<180 180 180>}, {limitMax=<180 180 180>}, {limitMax=<180 180 180>}, {limitMax=<180 180 180>}, {limitMax=<180 180 180>}]
    rotationOffsets [<0 0 0>, <0 0 0>, <0 0 0>, <0 0 0>, <0 0 0>]
    trailBone {boneName=dyn_hat_06}
   }
   n5 = Spring controller - Bone: dyn_hat_02!(ComponentPose:Component Space Pose Link Structure=@n4, Alpha:float=@n3) -> (Pose:Component Space Pose Link Structure) {
    is=title AnimGraphNode_SpringBone
    bLimitDisplacement=true
    bRotateX=true
    bRotateY=true
    bRotateZ=true
    bTranslateX=false
    alphaBoolBlend {blendOption=Linear}
    springBone {boneName=dyn_hat_02}
   }
   n6 = Trail controller - Bone: dyn_beard_10!(ComponentPose=@n5, Alpha=@n3, FakeVelocity:Vector="0,0,0") -> (Pose) {
    bUsePlanarLimit=true
    chainLength=8
    relaxationSpeedScale=2
    alphaScaleBiasClamp {bInterpResult=true, interpSpeedDecreasing=3}
    planarLimits [{drivingBone {boneName=spine_05}, planeTransform {location=<0 16 0>, rotation=<rot 0 -22.9999 89.9999>, scale=<1 1 1>}}]
    rotationLimits [
     {limitMax=<180 180 180>}
     {limitMax=<180 180 180>}
     {limitMax=<180 180 180>}
     {limitMax=<180 180 180>}
     {limitMax=<180 180 180>}
     {limitMax=<180 180 180>}
     {limitMax=<180 180 180>}
     {limitMax=<180 180 180>}
    ]
    rotationOffsets [<0 0 0>, <0 0 0>, <0 0 0>, <0 0 0>, <0 0 0>, <0 0 0>, <0 0 0>, <0 0 0>]
    trailBone {boneName=dyn_beard_10}
   }
   n7 = Spring controller - Bone: dyn_beard_02!(ComponentPose:Component Space Pose Link Structure=@n6, Alpha:float=@n3) -> (Pose:Component Space Pose Link Structure) {
    is=title AnimGraphNode_SpringBone
    bLimitDisplacement=true
    bRotateX=true
    bRotateY=true
    bRotateZ=true
    bTranslateX=false
    alphaBoolBlend {blendOption=Linear}
    springBone {boneName=dyn_beard_02}
   }
   n8 = RigidBody(ComponentPose=Local To Component(LocalPose=Use cached pose 'pre pose'())) {
    bClampLinearTranslationLimitToRefPose=true
    bOverrideWorldGravity=true
    overridePhysicsAsset=&/M_MED_Holiday_Time_FaceAcc_Dyn
    simulationSpace=BaseBoneSpace
    alphaBoolBlend {blendOption=Linear}
    baseBoneRef {boneName=neck_01}
    componentAppliedLinearAccClamp=<2000 2000 2000>
    componentLinearAccScale=<0.3 0.3 0.4>
    componentLinearVelScale=<0.2 0.2 0.2>
    overrideWorldGravity {z=-980}
   }
   n9 = Component To Local(ComponentPose=@n8)
   n10 = RigidBody(ComponentPose=Local To Component(LocalPose=Use cached pose 'pre pose'())) {
    bClampLinearTranslationLimitToRefPose=true
    bOverrideWorldGravity=true
    overridePhysicsAsset=&/M_MED_Holiday_Time_FaceAcc_Dyn_hatBackling
    simulationSpace=BaseBoneSpace
    alphaBoolBlend {blendOption=Linear}
    baseBoneRef {boneName=neck_01}
    componentAppliedLinearAccClamp=<2000 2000 2000>
    componentLinearAccScale=<0.25 0.35 0.35>
    componentLinearVelScale=<0.2 0.2 0.2>
    overrideWorldGravity {z=-980}
   }
   n11 = Component To Local(ComponentPose=@n10)
   n12 = Get isBackpackEquipped() -> (isBackpackEquipped:bool)
   n13 = RigidBody(ComponentPose=Local To Component(LocalPose=Blend Poses by bool(BlendPose_0=@n9, BlendPose_1=@n11, bActiveValue=@n12) {blendType=Linear})) {
    bClampLinearTranslationLimitToRefPose=true
    bOverrideWorldGravity=true
    overridePhysicsAsset=&/M_MED_Holiday_Time_FaceAcc_DynBeard
    simulationSpace=BaseBoneSpace
    alphaBoolBlend {blendOption=Linear}
    baseBoneRef {boneName=root}
    componentAppliedLinearAccClamp=<2000 2000 2000>
    componentLinearAccScale=<0.3 0.25 0.25>
    componentLinearVelScale=<0.3 0.3 0.3>
    overrideWorldGravity {z=-500}
   }
   Output Pose(Result=Blend Poses by bool(BlendPose_0=Component To Local(ComponentPose=@n13), BlendPose_1=Component To Local(ComponentPose=@n7), bActiveValue=Get bRigidBodyEnabled()) {blendType=Linear}) {layerGroup=None}
   n14 = Control Rig(Source=Copy Pose From Mesh() {bUseAttachedParent=true}) {
    alphaBoolBlend {blendOption=Linear}
    controlRigAssetReference {blueprintRigClass=&/M_MED_HolidayTime_Face_Acc_CtrlRig.M_MED_HolidayTime_Face_Acc_CtrlRig_C}
   }
   Save cached pose 'pre pose'(Pose=@n14) {cacheName=pre pose}
  ]}
 EventGraph {nodes [
   Event BlueprintUpdateAnimation()
   > n1 = Sequence!(execute) -> (then_0, then_1, then_2:Exec, then_3:Exec)
   Set trail ctrl alpha(
    execute=[@n1.then_0, @n2.else]
    trail ctrl alpha:double=MapRangeClamped(Value=Get PawnSpeed2D(), InRangeA=0, InRangeB=200, OutRangeA=1, OutRangeB=0)
   ) -> (Output_Get:double)
   n2 = Branch(
    execute=@n1.then_1
    Condition=OR Boolean!(
     A=Get bIsJumping()
     B=Get bIsFalling()
     C:bool=Get bIsSkydiveDiveMode()
     D:bool=Get bIsInVehicle() -> (bIsInVehicle:bool)
     E:bool=Get bIsInVehicleSeat() -> (bIsInVehicleSeat:bool)
    ) -> (ReturnValue)
   )
   > Set trail ctrl alpha(trail ctrl alpha=0.2)
   n3 = Get player_pawn() -> (player_pawn:Fort Player Pawn Athena Object Reference)
   Is Valid(exec=@n1.then_2, InputObject=@n3)
   > Do Once()
   > Set isBackpackEquipped(isBackpackEquipped:bool="Not Equal (Object)"!(A:Object Reference=GetSkeletalMeshForPartType(self=@n3, PartType=Backpack), B:Object Reference) -> (ReturnValue:bool)) -> (Output_Get:bool)
   Event BlueprintInitializeAnimation()
   n4 = TryGetPawnOwner()
   > Is Valid(InputObject=@n4)
   > n5 = Cast To FortPlayerPawnAthena(Object=@n4)
   > Set player_pawn(player_pawn:Fort Player Pawn Athena Object Reference=@n5) -> (Output_Get:Fort Player Pawn Athena Object Reference)
   n6 = GetPartSkeletalMeshComponent()
   OR Boolean!(
    A="Not Equal (Object)"(A=FindChildCommponentFromAttachSocket!(
      self:Custom Character Part Anim Instance Object Reference
      Parent:Skeletal Mesh Component Object Reference=@n6
      bIncludeAllDescendants:bool=false
      AttachSocket:Name=Cape
     ) -> (ReturnValue:Skeletal Mesh Component Object Reference))
    B="Not Equal (Object)"(A=FindChildCommponentFromAttachSocket(Parent=@n6, bIncludeAllDescendants=false, AttachSocket=backpack_br))
   ) -> (ReturnValue)
   Do Once()
  ]}
]
variables [
 isBackpackEquipped {type=boolean}
 player_pawn {type=FortPlayerPawnAthena}
 trail ctrl alpha {type=number}
]
